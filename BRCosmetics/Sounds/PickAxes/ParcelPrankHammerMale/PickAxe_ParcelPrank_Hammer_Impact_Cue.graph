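# Sound cue for the impact of the Parcel Prank hammer pickaxe, layering a hit sound with a prank laugh.
# A mixer plays one of six equally weighted impact waves alongside an equally weighted random laugh slot.
# That slot holds five laughs and one empty entry, so some hits have no laugh, and one laugh is lowered through a nested mixer.
# It uses pickaxe impact attenuation, Fort_Impacts class and melee weapon concurrency.
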SoundCue /BRCosmetics/Sounds/PickAxes/ParcelPrankHammerMale/PickAxe_ParcelPrank_Hammer_Impact_Cue
root=Mixer_0
nodes [
 {id=Mixer_0, children [Random_1,Random_0], properties {inputVolume [1,1]}}
 {id=Mixer_2, children [WavePlayer_14, null], properties {inputVolume [0.75,1]}}
 {id=Random_0, children [null,WavePlayer_1,WavePlayer_2,WavePlayer_3,WavePlayer_4,Mixer_2], properties {weights [1,1,1,1,1,1]}}
 {
  id=Random_1
  children [WavePlayer_6, WavePlayer_7, WavePlayer_8, WavePlayer_9, WavePlayer_10, WavePlayer_11]
  properties {weights [1,1,1,1,1,1]}
 }
 {id=WavePlayer_1, properties {soundWaveAssetPtr=&/PA_Parcel_Prank_Hammer_Impact_Laugh_1}}
 {id=WavePlayer_10, properties {soundWaveAssetPtr=&/PA_Parcel_Prank_Hammer_Impact_4}}
 {id=WavePlayer_11, properties {soundWaveAssetPtr=&/PA_Parcel_Prank_Hammer_Impact_5}}
 {id=WavePlayer_14, properties {soundWaveAssetPtr=&/PA_Parcel_Prank_Hammer_Impact_Laugh_6}}
 {id=WavePlayer_2, properties {soundWaveAssetPtr=&/PA_Parcel_Prank_Hammer_Impact_Laugh_2}}
 {id=WavePlayer_3, properties {soundWaveAssetPtr=&/PA_Parcel_Prank_Hammer_Impact_Laugh_3}}
 {id=WavePlayer_4, properties {soundWaveAssetPtr=&/PA_Parcel_Prank_Hammer_Impact_Laugh_4}}
 {id=WavePlayer_6, properties {soundWaveAssetPtr=&/PA_Parcel_Prank_Hammer_Impact_6}}
 {id=WavePlayer_7, properties {soundWaveAssetPtr=&/PA_Parcel_Prank_Hammer_Impact_1}}
 {id=WavePlayer_8, properties {soundWaveAssetPtr=&/PA_Parcel_Prank_Hammer_Impact_2}}
 {id=WavePlayer_9, properties {soundWaveAssetPtr=&/PA_Parcel_Prank_Hammer_Impact_3}}
]
properties {
 attenuationSettings=/Game/Sounds/Attenuation/Fort_Pickaxe_Impact_Att
 audioPropertiesSheet=/BRCosmetics/Sounds/Settings/PropertySheets/Pickaxe/APS_Cosmetic_Pickaxe_Impact
 bPrimeOnLoad=true
 maxDistance=4300
 soundClassObject=&Fort_Impacts
 volumeMultiplier=0.5625
 attenuationOverrides {
  bAttenuateWithLPF=true
  distanceAlgorithm=Logarithmic
  falloffDistance=800
  lPFRadiusMax=800
  lPFRadiusMin=400
  attenuationShapeExtents {x=300}
 }
 concurrencyOverrides {bLimitToOwner=true, maxCount=2, resolutionRule=StopOldest}
 concurrencySet [&Player_Melee_Weapon_Conc]
}
